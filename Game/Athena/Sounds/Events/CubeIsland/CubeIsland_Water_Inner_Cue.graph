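# CubeIsland_Water_Inner_Cue is a Sound Cue named for inner water audio at the Cube Island event.
# Its node graph is empty with no root, so it references no audio and produces no sound as stored.
# What remains is playback setup: lowered volume and pitch, the scripted-events sound class, non-spatialized attenuation overrides, and single-instance concurrency that blocks new plays.
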SoundCue /Game/Athena/Sounds/Events/CubeIsland/CubeIsland_Water_Inner_Cue
root=null
nodes []
properties {
 maxDistance=5000
 pitchMultiplier=0.8
 soundClassObject=/Game/Sounds/FortSoundClassesAndModes/Fort_Events_Scripted
 volumeMultiplier=0.2807
 attenuationOverrides {bSpatialize=false, falloffDistance=5000, nonSpatializedRadiusStart=500, stereoSpread=5000, attenuationShapeExtents {x=0}}
 concurrencyOverrides {maxCount=1, resolutionRule=PreventNew}
}
