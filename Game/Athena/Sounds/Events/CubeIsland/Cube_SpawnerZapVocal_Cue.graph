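# Sound cue named for a cube spawner's zap vocal in the Athena CubeIsland event, with an empty node graph and no root, so it references no audio.
# Only playback settings remain: the in-game mission sound class and a very distant Deimos spawner attenuation with a huge max distance.
# A concurrency override caps it at three instances, blocks new plays beyond that and applies volume scaling.
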SoundCue /Game/Athena/Sounds/Events/CubeIsland/Cube_SpawnerZapVocal_Cue
root=null
nodes []
properties {
 attenuationSettings=/Game/Sounds/Attenuation/Fort_Deimos_Spawnter_Very_Distant_Athena_Att
 bOverrideConcurrency=true
 maxDistance=1000000
 soundClassObject=/Game/Sounds/FortSoundClassesAndModes/Fort_Mission_IngameSounds
 volumeMultiplier=1
 concurrencyOverrides {maxCount=3, resolutionRule=PreventNew, volumeScale=0.7}
}
