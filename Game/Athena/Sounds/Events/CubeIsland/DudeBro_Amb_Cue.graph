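# Looping ambience cue for the CubeIsland event's DudeBro, blending close and distant drones, Butterfly build-up loops and randomly delayed distant thunder.
# Four attenuation layers cover close inner, focused and distant ranges, with the inner one reusing close branches at altered pitch.
# DrainAmount and DrainAmountInv parameters drive build-up and thunder volume and pitch.
# Looping envelopes slowly sweep pitch or volume, while quality levels swap in optimized waves or alternate thunder sets.
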SoundCue /Game/Athena/Sounds/Events/CubeIsland/DudeBro_Amb_Cue
root=Mixer_4
nodes [
 {
  id=Attenuation_0
  children [Mixer_0]
  properties {attenuationSettings=&/DudeBro_Large_Close_Inner_No_Focus_Att}
 }
 {
  id=Attenuation_10
  children [Mixer_9]
  properties {attenuationSettings=&/DudeBro_Large_CloseDistant_Focussed_No_Inner_Att}
 }
 {id=Attenuation_8, children [Enveloper_0], properties {attenuationSettings=&/DudeBro_Large_Distant_Att}}
 {
  id=Attenuation_9
  children [Enveloper_1]
  properties {attenuationSettings=&/DudeBro_Large_Close_Focussed_No_Inner_Cone_Att}
 }
 {id=Delay_1, children [Mixer_8], properties {delayMax=3, delayMin=0.5}}
 {
  id=Enveloper_0
  children [QualityLevel_8]
  properties {
   bLoop=true
   bLoopIndefinitely=true
   loopEnd=30
   pitchCurve {editorCurveData {keys [{time=0}, <at 15 0.75>, <at 20 1.2>, {time=30}]}}
   volumeCurve {editorCurveData {keys [<at 0 1>]}}
  }
 }
 {
  id=Enveloper_1
  children [QualityLevel_9]
  properties {
   bLoop=true
   bLoopIndefinitely=true
   loopEnd=30
   pitchCurve {editorCurveData {keys [{time=0}, <at 15 0.75>, <at 20 1.2>, {time=30}]}}
   volumeCurve {editorCurveData {keys [<at 0 1>]}}
  }
 }
 {
  id=Enveloper_2
  children [WavePlayer_23]
  properties {
   bLoop=true
   bLoopIndefinitely=true
   loopEnd=20
   pitchCurve {editorCurveData {keys [{time=0}, <at 10 0.5>, {time=20}]}}
   volumeCurve {editorCurveData {keys [<at 0 1>]}}
  }
 }
 {
  id=Enveloper_3
  children [Enveloper_2]
  properties {
   bLoop=true
   bLoopIndefinitely=true
   loopEnd=30
   pitchCurve {editorCurveData {keys [{time=0}]}}
   volumeCurve {editorCurveData {keys [<at 0 1>,<at 15 0.75>,<at 30 1>]}}
  }
 }
 {id=Looping_1, children [Delay_1]}
 {id=Mixer_0, children [Modulator_1,Modulator_0], properties {inputVolume [0.5,0.3]}}
 {id=Mixer_4, children [Attenuation_0, Attenuation_9, Attenuation_10, Attenuation_8], properties {inputVolume [1,1,1,1]}}
 {id=Mixer_6, children [Enveloper_3,Looping_1], properties {inputVolume [0.75,1]}}
 {id=Mixer_7, children [ModulatorContinuous_2], properties {inputVolume [2]}}
 {id=Mixer_8, children [QualityLevel_11], properties {inputVolume [0.4]}}
 {id=Mixer_9, children [Mixer_7, ModulatorContinuous_1], properties {inputVolume [1,1]}}
 {
  id=ModulatorContinuous_0
  children [QualityLevel_10]
  properties {pitchModulationParams {maxOutput=1.2, minOutput=0.8}, volumeModulationParams {minOutput=0.1, parameterName=DrainAmountInv}}
 }
 {
  id=ModulatorContinuous_1
  children [Mixer_6]
  properties {pitchModulationParams {maxOutput=1.2, minOutput=0.8, parameterName=DrainAmount}, volumeModulationParams {parameterName=DrainAmount}}
 }
 {
  id=ModulatorContinuous_2
  children [ModulatorContinuous_0]
  properties {pitchModulationParams {maxOutput=1.2, minOutput=0.5, parameterName=DrainAmount}}
 }
 {id=Modulator_0, children [Mixer_9], properties {pitchMax=0.8, pitchMin=0.8}}
 {id=Modulator_1, children [Enveloper_1], properties {pitchMax=0.95}}
 {id=QualityLevel_10, children [WavePlayer_24, WavePlayer_24, null, null]}
 {id=QualityLevel_11, children [Random_0,Random_1,null,null]}
 {id=QualityLevel_8, children [WavePlayer_26, WavePlayer_26, WavePlayer_29, WavePlayer_29]}
 {id=QualityLevel_9, children [WavePlayer_28, WavePlayer_28, WavePlayer_30, WavePlayer_30]}
 {
  id=Random_0
  children [WavePlayer_27, WavePlayer_31, WavePlayer_32, WavePlayer_33, WavePlayer_34, WavePlayer_35, WavePlayer_36, WavePlayer_37, WavePlayer_38]
  properties {weights [1,1,1,1,1,1,1,1,1]}
 }
 {id=Random_1, children [WavePlayer_39, WavePlayer_40, WavePlayer_41], properties {weights [1,1,1]}}
 {id=WavePlayer_23, properties {bLooping=true, soundWaveAssetPtr=/Game/Athena/Sounds/Events/Butterfly/Butterfly_BuildUp_2}}
 {id=WavePlayer_24, properties {bLooping=true, soundWaveAssetPtr=/Game/Athena/Sounds/Events/Butterfly/Butterfly_BuildUp_1}}
 {id=WavePlayer_26, properties {bLooping=true, soundWaveAssetPtr=&/DudeBro_Amb_Rne_Dist_01}}
 {id=WavePlayer_27, properties {soundWaveAssetPtr=/Game/Sounds/Fort_GamePlay_Sounds/Storms/Dist_Thunder_001}}
 {id=WavePlayer_28, properties {bLooping=true, soundWaveAssetPtr=&/DudeBro_Amb_Rne_Close_01}}
 {id=WavePlayer_29, properties {bLooping=true, soundWaveAssetPtr=&/DudeBro_Amb_Rne_Dist_01_opt}}
 {id=WavePlayer_30, properties {bLooping=true, soundWaveAssetPtr=&/DudeBro_Amb_Rne_Close_01_opt}}
 {id=WavePlayer_31, properties {soundWaveAssetPtr=/Game/Sounds/Fort_GamePlay_Sounds/Storms/Dist_Thunder_002}}
 {id=WavePlayer_32, properties {soundWaveAssetPtr=/Game/Sounds/Fort_GamePlay_Sounds/Storms/Dist_Thunder_003}}
 {id=WavePlayer_33, properties {soundWaveAssetPtr=/Game/Sounds/Fort_GamePlay_Sounds/Storms/Dist_Thunder_004}}
 {id=WavePlayer_34, properties {soundWaveAssetPtr=/Game/Sounds/Fort_GamePlay_Sounds/Storms/Dist_Thunder_005}}
 {id=WavePlayer_35, properties {soundWaveAssetPtr=/Game/Sounds/Fort_GamePlay_Sounds/Storms/Dist_Thunder_006}}
 {id=WavePlayer_36, properties {soundWaveAssetPtr=/Game/Sounds/Fort_GamePlay_Sounds/Storms/Dist_Thunder_007}}
 {id=WavePlayer_37, properties {soundWaveAssetPtr=/Game/Sounds/Fort_GamePlay_Sounds/Storms/Dist_Thunder_008}}
 {id=WavePlayer_38, properties {soundWaveAssetPtr=/Game/Sounds/Fort_GamePlay_Sounds/Storms/Dist_Thunder_009}}
 {id=WavePlayer_39, properties {soundWaveAssetPtr=&Dist_Thunder_001}}
 {id=WavePlayer_40, properties {soundWaveAssetPtr=&Dist_Thunder_002}}
 {id=WavePlayer_41, properties {soundWaveAssetPtr=&Dist_Thunder_003}}
]
properties {maxDistance=2097152, soundClassObject=/Game/Sounds/FortSoundClassesAndModes/Fort_Events_Scripted, volumeMultiplier=0.5}
