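# DudeBro_Amb_Music_Cue is a SoundCue in the CubeIsland event sound folder, named for ambient music.
# It has no root node and an empty node list, so no sound source is wired in.
# It keeps a boosted volume multiplier and overrides attenuation with spatialization off and its own falloff distance.
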SoundCue /Game/Athena/Sounds/Events/CubeIsland/DudeBro_Amb_Music_Cue
root=null
nodes []
properties {
 bOverrideAttenuation=true
 maxDistance=2097152
 volumeMultiplier=1.65
 attenuationOverrides {bSpatialize=false, falloffDistance=5000, attenuationShapeExtents {x=0}}
}
