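# DudeBro_Amb_Zap_Cue is a SoundCue filed with the CubeIsland event sounds, named as an ambient zap.
# It has no nodes and a null root, so it contains no audio to play.
# It keeps only its settings: the scripted-events sound class, reduced volume, a max distance, and a concurrency override that caps instances, stops the oldest and applies volume scaling.
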SoundCue /Game/Athena/Sounds/Events/CubeIsland/DudeBro_Amb_Zap_Cue
root=null
nodes []
properties {
 bOverrideConcurrency=true
 maxDistance=7000
 soundClassObject=/Game/Sounds/FortSoundClassesAndModes/Fort_Events_Scripted
 volumeMultiplier=0.1404
 concurrencyOverrides {maxCount=4, resolutionRule=StopOldest, volumeScale=0.75}
}
