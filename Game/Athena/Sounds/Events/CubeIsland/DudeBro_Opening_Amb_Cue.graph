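# DudeBro_Opening_Amb_Cue is a SoundCue in the CubeIsland event sounds, named as an opening ambience, that holds no sound nodes and has no root.
# As written it has nothing to play, leaving only its cue-level settings.
# It is assigned to the scripted events sound class and a map-wide, sky-object focus attenuation with a very large max distance.
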SoundCue /Game/Athena/Sounds/Events/CubeIsland/DudeBro_Opening_Amb_Cue
root=null
nodes []
properties {
 attenuationSettings=/Game/Athena/Sounds/Events/Attenuations/Events_Mapwide_Focus_SkyObject_Att
 maxDistance=45000
 soundClassObject=/Game/Sounds/FortSoundClassesAndModes/Fort_Events_Scripted
 volumeMultiplier=1
}
