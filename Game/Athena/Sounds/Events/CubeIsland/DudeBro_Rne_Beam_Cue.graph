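# Sound cue for the DudeBro rune beam in the Athena CubeIsland event, routed to the scripted events sound class.
# It holds no sound nodes and no root, so it defines only playback settings and produces no audio as written.
# Attenuation overrides set logarithmic falloff, distance low-pass filtering, a stereo spread and interpolated listener focus that lowers volume off-axis.
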SoundCue /Game/Athena/Sounds/Events/CubeIsland/DudeBro_Rne_Beam_Cue
root=null
nodes []
properties {
 maxDistance=7000
 soundClassObject=/Game/Sounds/FortSoundClassesAndModes/Fort_Events_Scripted
 volumeMultiplier=1.25
 attenuationOverrides {
  bAttenuateWithLPF=true
  bEnableFocusInterpolation=true
  bEnableListenerFocus=true
  distanceAlgorithm=Logarithmic
  falloffDistance=6500
  focusAttackInterpSpeed=2
  focusAzimuth=90
  focusReleaseInterpSpeed=2
  lPFFrequencyAtMax=500
  lPFRadiusMin=2000
  nonFocusAzimuth=120
  nonFocusVolumeAttenuation=0.2
  stereoSpread=1000
  attenuationShapeExtents {x=0}
 }
}
