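# Sound cue named for a rune beam ground effect in the CubeIsland event, assigned to the scripted events sound class.
# The node graph is empty with no root, so the cue references no audio and produces no sound as it stands.
# It retains a reduced volume and attenuation overrides with distance falloff, stereo spread and a low-pass filter that muffles with distance.
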SoundCue /Game/Athena/Sounds/Events/CubeIsland/DudeBro_Rne_Beam_Ground_Cue
root=null
nodes []
properties {
 maxDistance=10000
 soundClassObject=/Game/Sounds/FortSoundClassesAndModes/Fort_Events_Scripted
 volumeMultiplier=0.5263
 attenuationOverrides {
  bAttenuateWithLPF=true
  falloffDistance=4000
  lPFFrequencyAtMax=500
  lPFRadiusMax=4000
  lPFRadiusMin=2000
  stereoSpread=1000
  attenuationShapeExtents {x=0}
 }
}
