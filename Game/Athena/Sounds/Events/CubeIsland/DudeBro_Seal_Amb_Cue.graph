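# DudeBro_Seal_Amb_Cue is a SoundCue filed under the Athena CubeIsland event sounds, named as a seal ambience.
# Its graph has no nodes and no root, so the cue produces no audio as saved.
# It keeps its playback setup: the scripted-events sound class, overridden attenuation with stereo spread, and a concurrency override that caps simultaneous instances at two by stopping the quietest.
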SoundCue /Game/Athena/Sounds/Events/CubeIsland/DudeBro_Seal_Amb_Cue
root=null
nodes []
properties {
 bOverrideConcurrency=true
 maxDistance=7000
 soundClassObject=/Game/Sounds/FortSoundClassesAndModes/Fort_Events_Scripted
 volumeMultiplier=1.0025
 attenuationOverrides {falloffDistance=5000, stereoSpread=512, attenuationShapeExtents {x=0}}
 concurrencyOverrides {maxCount=2, resolutionRule=StopQuietest}
}
